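# Niagara boost trail for the Umbra Genesis vehicle cosmetic, emitting from skeletal-mesh sockets for each boost port.
# Cone clouds form short-lived expanding flipbook puffs, while fast rearward ribbons render in five layers with red and black color parameters.
# A third emitter scatters curl-noise sprites and spinning meshes.
# Spawn groups vary size, velocity and ribbon width within emitters; painted-variant color overrides are present but disabled.
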NiagaraSystem /VehicleCosmetics/Boosts/Boost_UmbraGenesis/FX/NS_UmbraGenesis
dependencies {
 usedDataInterfaces [
  &NiagaraDataInterfaceCurve
  &NiagaraDataInterfaceSpriteRendererInfo
  &NiagaraDataInterfaceArrayFloat
  &NiagaraDataInterfaceSkeletalMesh
 ]
 usedDynamicInputs []
 usedModules [
  &self:SystemSpawnScript.SystemScriptSource.SystemScriptGraph.NiagaraNodeAssignment_0.SetVariables_#1
  &SystemState
 ]
}
emitters {
 NE_Cone_Clouds {
  emitterUpdateScript [
   &EmitterState(Inactive Response=::NewEnumerator1, Life Cycle Mode=::NewEnumerator1, UseLoopDelay=false)
   &VC_FilteredSocketCount()
   &SpawnRate(SpawnRate=30, Use Spawn Probability=false)
   &SpawnRate(SpawnGroup=1, SpawnRate=35, Use Spawn Probability=false) {module=SpawnRate001}
  ]
  particleSpawnScript [
   &VC_PaintedVariantsOverride(
    Black=<rgba 1 1 1 1>
    BurntSienna=<rgba 0.5 0.1058 0 1>
    Cobalt=<rgba 0.429 0.4488 1.5 1>
    Crimson=<rgba 1.5 0.3 0.3325 1>
    DefaultColor=<rgba 1 1 1 0.9217>
    ForestGreen=<rgba 0.3464 0.5 0.3541 1>
    Grey=<rgba 0.9 0.9 0.9 0.8>
    Lime=<rgba 0.25 0.5 0 1>
    Orange=<rgba 1 0.335 0.05 1>
    Pink=<rgba 1 0.5 0.8182 1>
    Purple=<rgba 0.6952 0.3438 1.2 1>
    SkyBlue=<rgba 0.2 0.7323 1 1>
   ) {enabled=false}
   /SetVariables(Particles.VisibilityTag=$Emitter.SpawnGroup) {isSetParameters=true, module=SetVariables_#2}
   &InitializeParticle(
    Color Mode=::NewEnumerator0
    Lifetime=&MaskFloatBySpawnGroup(
     Masked Float=&RandomRangeFloat(Minimum=0.3, Maximum=0.5)
     Passthrough Float=&MaskFloatBySpawnGroup(Masked Float=&RandomRangeFloat(Minimum=0.3, Maximum=0.7), Passthrough Float=0)
     Spawn Group Mask=1
    )
    Sprite Rotation Angle=&RandomRangeFloat(Minimum=0, Maximum=1)
    Sprite Rotation Mode=::NewEnumerator1
    Sprite Size Mode=::NewEnumerator3
    Uniform Sprite Size=50
    UsePositionOffset=false
    Write Material Random=true
   )
   &PartitionParticles(Number Of Partitions=$Emitter.PortCount)
   &SkeletalMeshLocation(
    Bone / Socket Index=$Particles.PartitionParticles.Partition
    Bone / Socket Sampling Mode=::NewEnumerator5
    MaskSpawnGroup=false
    Mesh Sampling Type=::NewEnumerator6
    Sampled Position Offset=&AddVectorToPosition(Position=<-65 0 0>, Vector=$User.DTCarOffset001)
    Skeletal Mesh=$System.SkeletalMesh
   )
   &AddVelocity(Velocity=&MaskVectorBySpawnGroup(
     Masked Vector=&RandomRangeVector(Minimum=<0 0 0>, Maximum=<-150 0 0>)
     Passthrough Vector=&MaskVectorBySpawnGroup(
      Masked Vector=&RandomRangeVector(Minimum=<-10 0 0>, Maximum=<0 1 1>)
      Passthrough Vector=<0 0 0>
      Spawn Group Mask=1
     )
    ))
   &ShapeLocation(
    MaskSpawnGroup=false
    Offset Coordinate Space=::Local
    Offset Mode=::NewEnumerator0
    Override Seed=false
    Sphere Radius=&MaskFloatBySpawnGroup(
     Masked Float=35
     Passthrough Float=&MaskFloatBySpawnGroup(Masked Float=15, Passthrough Float=0, Spawn Group Mask=1)
    )
    UseOwnerScale=true
   )
  ]
  particleUpdateScript [
   &ParticleState()
   &ScaleSpriteSize(
    Scale Sprite Size Mode=::NewEnumerator0
    Uniform Scale Factor=&MaskFloatBySpawnGroup(
     Masked Float=&Multiply_Float(
      A=2.5
      B=&FloatFromCurve(FloatCurve=DataInterface() {propertyValues {Curve {defaultValue=1, keys [<at 0 0 RCIM_Cubic>,<at 0.3 1 RCIM_Cubic>]}}}, Scale Curve=0.7)
     )
     Passthrough Float=&MaskFloatBySpawnGroup(
      Masked Float=&Multiply_Float(
       A=3
       B=&FloatFromCurve(FloatCurve=DataInterface() {propertyValues {Curve {defaultValue=1, keys [<at 0 0 RCIM_Cubic>,<at 1 1 RCIM_Cubic>]}}}, Scale Curve=0.9)
      )
      Passthrough Float=0
      Spawn Group Mask=1
     )
    )
   )
   &SolveForcesAndVelocity(Clamp Velocity=false, Limit Acceleration=false)
   &JitterPosition(Jitter Amount:NiagaraFloat=&RandomRangeFloat(Minimum=-4, Maximum=4), Jitter Delay=0.6)
   &SubUVAnimation(
    "Sprite Renderer "=DataInterface() {propertyValues {SpriteRenderer=&self:NE_Cone_Clouds_0.NiagaraSpriteRendererProperties_0}}
    SubUV Animation Mode:ENiagara_SubUVLookupModeV2=::NewEnumerator1
    SubUV Curve=DataInterface() {propertyValues {Curve {
       defaultValue=0
       keys [{interpMode=RCIM_Cubic, tangentMode=RCTM_Break, time=0, value=0}, {interpMode=RCIM_Cubic, tangentMode=RCTM_Break, time=1, value=1}]
      }}}
    UseEndFrame=false
    UseStartFrame=false
   )
   &SubUVAnimation(
    Direct Frame Index:NiagaraFloat=&RandomRangeFloat(Minimum=0, Maximum=2)
    "Sprite Renderer "=DataInterface() {propertyValues {SpriteRenderer=&self:NE_Cone_Clouds_0.NiagaraSpriteRendererProperties_0}}
    SubUV Animation Mode:ENiagara_SubUVLookupModeV2=::NewEnumerator5
   ) {enabled=false, module=SubUVAnimation002}
   &CurlNoiseForce(
    Bias Noise=false
    Noise Frequency=50000
    Noise Strength=&RandomRangeFloat(Minimum=95, Maximum=100)
    Pan Noise=false
    Randomization Vector:Vector3f=<0.65 0.125 0>
   ) {enabled=false}
   &InheritSourceMovement(
    Applied Position Scale=&VectorFromFloat(Value=&FloatFromCurve(FloatCurve=DataInterface() {propertyValues {Curve {keys [{interpMode=RCIM_Cubic, tangentMode=RCTM_Break, time=0.3, value=1}, {interpMode=RCIM_Cubic, tangentMode=RCTM_Break, time=1, value=0}]}}}))
    ApplyForce=false
    ApplyPosition=true
    ApplyPositionFalloff=false
   ) {enabled=false}
   &Drag(
    Drag=&RandomRangeFloat(
     Minimum=0
     Maximum=&FloatFromCurve(FloatCurve=DataInterface() {propertyValues {Curve {
         defaultValue=1
         keys [
          {arriveTangent=0.0905, interpMode=RCIM_Cubic, leaveTangent=0.0905, tangentMode=RCTM_User, time=0.2, value=0}
          {arriveTangent=1.0887, interpMode=RCIM_Cubic, leaveTangent=1.0887, tangentMode=RCTM_User, time=1, value=0.25}
         ]
        }}})
    )
    Use Linear Drag=true
    Use Rotational Drag=true
   ) {enabled=false}
   &DynamicMaterialParameters(
    Index 0 Param 2=3
    Index 0 Param 3=&FloatFromCurve(FloatCurve=DataInterface() {propertyValues {Curve {keys [<at 0 0>,<at 1 15>]}}})
    Module.Index0Float0Write=true
    Module.Index0Float1Write=true
    Module.Index0Float2Write=true
    Module.Index0Float3Write=true
   ) {enabled=false}
   /SetVariables(Particles.ConeCLoudsAnim:NiagaraFloat=&RandomRangeFloat(Minimum=0, Maximum=2)) {isSetParameters=true}
  ]
  rendererClasses [&NiagaraSpriteRendererProperties]
  renderers [
   {rendererClass=&NiagaraSpriteRendererProperties}
   {rendererClass=&NiagaraSpriteRendererProperties, rendererIndex=1}
   {rendererClass=&NiagaraSpriteRendererProperties, rendererIndex=2}
  ]
 }
 NE_Ribbon {
  emitterUpdateScript [
   &EmitterState(Life Cycle Mode=::NewEnumerator1, Loop Duration=1, UseLoopDelay=false)
   &VC_FilteredSocketCount()
   &VC_RateQualityHandler(Epic Spawn Rate=30, High Spawn Rate=30, Low Spawn Rate=30, Use Custom Rate?=true)
   &VC_SpawnRibbon_v1(Number of Ribbons=$Emitter.PortCount, Spawn Time=1, SpawnRate=$Emitter.Spawn Rate)
  ]
  particleSpawnScript [
   &VC_PaintedVariantsOverride(
    DefaultColor=<rgba 0.5104 0 0.002 1>
    R1 BurntSienna:LinearColor=<rgba 0.0642 0.0198 0.0079 1>
    R1 DefaultColor=<rgba 0 0 0 1>
    Use R1 Painted Overrides=true
   ) {enabled=false}
   &VC_InitializeRibbon_v1(
    Color=$StackContext.Color
    Custom Offset=true
    Enable Skeletal Position=true
    Lifetime=1.25
    Local Vector Offset=&AddVectorToPosition(Position=<-25 0 0>, Vector=Linked() {linkedVariable {name=User.DTCarOffset, type=NiagaraPosition}})
    Offset Amount=1
    Ribbon Width=60
   )
   &AddVelocity(Velocity=<-1050 0 0>) {module=AddVelocity001}
   /SetVariables() {isSetParameters=true}
   &ShapeLocation(MaskSpawnGroup=false, Override Seed=false, Sphere Radius=0.5, UseOwnerScale=true) {enabled=false}
  ]
  particleUpdateScript [
   &ParticleState()
   &InheritSourceMovement(
    Applied Position Scale=&VectorFromFloat(Value=&FloatFromCurve(FloatCurve=DataInterface() {propertyValues {Curve {keys [
          {interpMode=RCIM_Cubic, tangentMode=RCTM_Break, time=0, value=1}
          {arriveTangent=-2.9025, interpMode=RCIM_Cubic, tangentMode=RCTM_Break, time=1, value=0}
         ]}}}))
    ApplyForce=false
    ApplyPosition=true
    ApplyPositionFalloff=false
   )
   /SetVariables(
    Particles.Ribbon_BlackColor:LinearColor=&Multiply_LinearColorByFloat(
     Linear Color=<rgba 0 0 0 10>
     Float=&FloatFromCurve(FloatCurve=DataInterface() {propertyValues {Curve {keys [<at 0 0>,<at 0.2 1>,<at 2 0.2>]}}})
    )
    Particles.Ribbon_Red:LinearColor=<rgba 1 0 0.004 10>
   ) {isSetParameters=true}
   &SolveForcesAndVelocity(Clamp Velocity=false, Limit Acceleration=false)
   &VC_UpdateRibbon_v1(Enable Skeletal Position=true)
   /SetVariables(
    Particles.R1Width:NiagaraFloat=&SelectFloatFromArray(Direct Array Index=&SpawnGroup(), Float Selection Array=DataInterface() {propertyValues {FloatData [0,50]}})
    Particles.R2Width:NiagaraFloat=&SelectFloatFromArray(Direct Array Index=&SpawnGroup(), Float Selection Array=DataInterface() {propertyValues {FloatData [-50,0]}})
    Particles.R3Width:NiagaraFloat=&SelectFloatFromArray(Direct Array Index=&SpawnGroup(), Float Selection Array=DataInterface() {propertyValues {FloatData [0,64]}})
    Particles.R4Width:NiagaraFloat=&SelectFloatFromArray(Direct Array Index=&SpawnGroup(), Float Selection Array=DataInterface() {propertyValues {FloatData [-64,0]}})
   ) {isSetParameters=true}
   /SetVariables(
    Particles.R1Width:NiagaraFloat=&Multiply_Float(
     A=$Particles.Initial.R1Width
     B=&FloatFromCurve(
      FloatCurve=DataInterface() {propertyValues {Curve {defaultValue=0.3, keys [{tangentMode=RCTM_User, time=0.5, value=0.5}, {tangentMode=RCTM_User, time=0.8, value=1}]}}}
      Scale Curve=1.5
     )
    )
    Particles.R2Width:NiagaraFloat=&Multiply_Float(
     A=$Particles.Initial.R2Width
     B=&FloatFromCurve(
      FloatCurve=DataInterface() {propertyValues {Curve {defaultValue=0.3, keys [{tangentMode=RCTM_User, time=0.5, value=0.5}, {tangentMode=RCTM_User, time=0.8, value=1}]}}}
      Scale Curve=1.5
     )
    )
    Particles.R3Width:NiagaraFloat=&Multiply_Float(
     A=$Particles.Initial.R3Width
     B=&FloatFromCurve(
      FloatCurve=DataInterface() {propertyValues {Curve {defaultValue=0.3, keys [{tangentMode=RCTM_User, time=0.5, value=0.5}, {tangentMode=RCTM_User, time=0.8, value=1}]}}}
      Scale Curve=1.5
     )
    )
    Particles.R4Width:NiagaraFloat=&Multiply_Float(
     A=$Particles.Initial.R4Width
     B=&FloatFromCurve(
      FloatCurve=DataInterface() {propertyValues {Curve {defaultValue=0.3, keys [{tangentMode=RCTM_User, time=0.5, value=0.5}, {tangentMode=RCTM_User, time=0.8, value=1}]}}}
      Scale Curve=1.5
     )
    )
   ) {enabled=false, isSetParameters=true}
   &CurlNoiseForce(
    Noise Frequency=300
    Noise Strength=&FloatFromCurve(
     FloatCurve=DataInterface() {propertyValues {Curve {keys [
         {arriveTangent=1.34, interpMode=RCIM_Cubic, leaveTangent=1.34, tangentMode=RCTM_User, time=0.1, value=0}
         {arriveTangent=1.1111, interpMode=RCIM_Cubic, leaveTangent=1.1111, time=0.2, value=1}
         <at 1 1 RCIM_Cubic>
        ]}}}
     Scale Curve=250
    )
    Pan Noise=true
    Pan Noise Field=<1 1 1>
   ) {enabled=false}
   &JitterPosition(
    Jitter Amount=0.2
    Jitter Delay=0.6
    Jitter Offset:Vector3f=&VectorFromFloat(Value=&RandomRangeFloat(Minimum=-1, Maximum=1))
   ) {enabled=false}
   /SetVariables(Particles.Position=&AddVectorToPosition(
     Position=$Particles.Position
     Vector=/VehicleCosmetics/FX/Niagara/Modules/VC_ConvertVectorToWorld(Vector:Vector3f=&MakeVector(
       X=0
       Y=&RandomRangeFloat(Minimum=-0.04, Maximum=0.04)
       Z=&RandomRangeFloat(Minimum=-0.04, Maximum=0.04)
      ))
    )) {enabled=false, isSetParameters=true}
   /SetVariables(Particles.RibbonWidth1:NiagaraFloat=&Multiply_Float(A=50, B=1), Particles.RibbonWidth2=62) {enabled=false, isSetParameters=true}
  ]
  rendererClasses [&NiagaraRibbonRendererProperties]
  renderers [
   {rendererClass=&NiagaraRibbonRendererProperties}
   {rendererClass=&NiagaraRibbonRendererProperties, rendererIndex=1}
   {rendererClass=&NiagaraRibbonRendererProperties, rendererIndex=2}
   {rendererClass=&NiagaraRibbonRendererProperties, rendererIndex=3}
   {rendererClass=&NiagaraRibbonRendererProperties, rendererIndex=4}
  ]
 }
 NE_Sprites {
  emitterUpdateScript [
   &EmitterState()
   &SpawnRate(SpawnGroup=4, SpawnRate=2, Use Spawn Probability=false) {module=SpawnRate004}
   &VC_FilteredSocketCount()
   &SpawnRate(SpawnRate=4, Use Spawn Probability=false)
   &SpawnRate(SpawnGroup=1, SpawnRate=6, Use Spawn Probability=true) {module=SpawnRate001}
   &SpawnRate(SpawnGroup=2, SpawnRate=4, Use Spawn Probability=true) {module=SpawnRate002}
   &SpawnRate(SpawnGroup=3, SpawnRate=3, Use Spawn Probability=false) {module=SpawnRate003}
  ]
  particleSpawnScript [
   &VC_PaintedVariantsOverride(
    Black=<rgba 1 1 1 1>
    BurntSienna=<rgba 0.5 0.1058 0 1>
    Cobalt=<rgba 0.429 0.4488 1.5 1>
    Crimson=<rgba 1.5 0.3 0.3325 1>
    ForestGreen=<rgba 0.3464 0.5 0.3541 1>
    Grey=<rgba 0.9 0.9 0.9 0.8>
    Lime=<rgba 0.25 0.5 0 1>
    Orange=<rgba 1 0.335 0.05 1>
    Pink=<rgba 1 0.5 0.8182 1>
    Purple=<rgba 0.6952 0.3438 1.2 1>
    R1 DefaultColor=<rgba 0.1615 0.0005 0 1>
    SkyBlue=<rgba 0.2 0.7323 1 1>
    Use R1 Painted Overrides=true
    Use R2 Painted Overrides=true
   ) {enabled=false}
   /SetVariables(Particles.VisibilityTag=$Emitter.SpawnGroup) {isSetParameters=true, module=SetVariables_#2}
   &InitializeParticle(
    Lifetime=&RandomRangeFloat(Minimum=0.6, Maximum=1.1)
    Mesh Scale Mode=::NewEnumerator3
    Mesh Uniform Scale=0.1
    Sprite Rotation Angle=&RandomRangeFloat(Minimum=1, Maximum=5)
    Sprite Rotation Mode=::NewEnumerator1
    Sprite Size Mode=::NewEnumerator3
    Uniform Sprite Size=&MaskFloatBySpawnGroup(
     Masked Float=15
     Passthrough Float=&MaskFloatBySpawnGroup(
      Masked Float=&RandomRangeFloat(Minimum=5, Maximum=8)
      Passthrough Float=&MaskFloatBySpawnGroup(
       Masked Float=&RandomRangeFloat(Minimum=5, Maximum=8)
       Passthrough Float=&MaskFloatBySpawnGroup(
        Masked Float=&RandomRangeFloat(Minimum=5, Maximum=8)
        Passthrough Float=&RandomRangeFloat(Minimum=10, Maximum=15)
        Spawn Group Mask=3
       )
       Spawn Group Mask=2
      )
      Spawn Group Mask=1
     )
    )
    UsePositionOffset=false
    Write Material Random=true
   )
   &PartitionParticles(Number Of Partitions=$Emitter.PortCount)
   &SkeletalMeshLocation(
    Bone / Socket Index=$Particles.PartitionParticles.Partition
    Bone / Socket Sampling Mode=::NewEnumerator5
    MaskSpawnGroup=false
    Mesh Sampling Type=::NewEnumerator6
    Sampled Position Offset=&AddVectorToPosition(Position=<-100 0 0>, Vector=$User.DTCarOffset001)
    Skeletal Mesh=$System.SkeletalMesh
   )
   &AddVelocity(
    Velocity=&MaskVectorBySpawnGroup(
     Masked Vector=&RandomRangeVector(Minimum=<-250 0 -5>, Maximum=<-300 0 5>)
     Passthrough Vector=&RandomRangeVector(Minimum=<-300 0 -5>, Maximum=<-450 0 5>)
     Spawn Group Mask=4
    )
    Velocity Speed Scale=0.8
   )
   &ShapeLocation(
    MaskSpawnGroup=false
    Offset Coordinate Space=::Local
    Offset Mode=::NewEnumerator0
    Override Seed=false
    Sphere Radius=60
    UseOwnerScale=true
   )
   &InitialMeshOrientation(Use Rotation Vector=false)
  ]
  particleUpdateScript [
   &ParticleState()
   &UpdateMeshOrientation(Rotation Rate=0.5, Rotation Vector=<1 1 1>)
   &CurlNoiseForce(
    Bias Noise=false
    Noise Frequency=50000
    Noise Strength=&RandomRangeFloat(Minimum=100, Maximum=200)
    Pan Noise=false
    Randomization Vector:Vector3f=<0.65 0.125 0>
   )
   &InheritSourceMovement(
    Applied Position Scale=&VectorFromFloat(Value=&FloatFromCurve(FloatCurve=DataInterface() {propertyValues {Curve {keys [{interpMode=RCIM_Cubic, tangentMode=RCTM_Break, time=0.3, value=1}, {interpMode=RCIM_Cubic, tangentMode=RCTM_Break, time=1, value=0}]}}}))
    ApplyForce=false
    ApplyPosition=true
    ApplyPositionFalloff=false
   )
   &ScaleSpriteSize(
    Scale Sprite Size Mode=::NewEnumerator0
    Uniform Scale Factor=&Multiply_Float(
     A=3
     B=&FloatFromCurve(FloatCurve=DataInterface() {propertyValues {Curve {defaultValue=1, keys [<at 0 0.5 RCIM_Cubic>,<at 0.35 1 RCIM_Cubic>]}}})
    )
   )
   &Drag(
    Drag=&RandomRangeFloat(
     Minimum=0
     Maximum=&FloatFromCurve(FloatCurve=DataInterface() {propertyValues {Curve {
         defaultValue=1
         keys [
          {arriveTangent=0.0905, interpMode=RCIM_Cubic, leaveTangent=0.0905, tangentMode=RCTM_User, time=0.2, value=0}
          {arriveTangent=1.0887, interpMode=RCIM_Cubic, leaveTangent=1.0887, tangentMode=RCTM_User, time=1, value=0.25}
         ]
        }}})
    )
    Use Linear Drag=true
    Use Rotational Drag=true
   )
   &SolveForcesAndVelocity(Clamp Velocity=false, Limit Acceleration=false)
   &DynamicMaterialParameters(
    Index 0 Param 2=3
    Index 0 Param 3=&FloatFromCurve(FloatCurve=DataInterface() {propertyValues {Curve {keys [<at 0 0>,<at 1 15>]}}})
    Module.Index0Float0Write=true
    Module.Index0Float1Write=true
    Module.Index0Float2Write=true
    Module.Index0Float3Write=true
   ) {enabled=false}
   /SetVariables(Particles.StarRed:LinearColor=<rgba 0.5104 0 0.002 1>) {isSetParameters=true}
   &JitterPosition(Jitter Amount:NiagaraFloat=&RandomRangeFloat(Minimum=-5, Maximum=5), Jitter Delay=0.6)
   &SubUVAnimation(
    "Sprite Renderer "=DataInterface() {propertyValues {SpriteRenderer=&self:NE_Sprites_0.NiagaraSpriteRendererProperties_2}}
    SubUV Animation Mode:ENiagara_SubUVLookupModeV2=::NewEnumerator1
    SubUV Curve=DataInterface() {propertyValues {Curve {defaultValue=0, keys [{tangentMode=RCTM_User, time=0.1, value=0}, {tangentMode=RCTM_User, time=0.8, value=1}]}}}
    UseEndFrame=false
    UseStartFrame=false
   ) {module=SubUVAnimation002}
   /SetVariables(Particles.Staticcloud:NiagaraFloat=&RandomRangeFloat(Minimum=0, Maximum=2)) {isSetParameters=true}
   &DynamicMaterialParameters(
    Index 0 Param 1=1
    Module.Index0Float0Write=true
    Module.Index0Float1Write=true
    Module.Index0Float2Write=true
    Module.Index0Float3Write=true
    Module.Index1Float0Write=true
    Module.Index1Float1Write=true
    Module.Index1Float2Write=true
    Module.Index1Float3Write=true
    Write Parameter Index 1=::NewEnumerator1
   ) {module=DynamicMaterialParameters001}
  ]
  rendererClasses [&NiagaraSpriteRendererProperties, &NiagaraMeshRendererProperties]
  renderers [
   {rendererClass=&NiagaraSpriteRendererProperties}
   {rendererClass=&NiagaraSpriteRendererProperties, rendererIndex=1}
   {rendererClass=&NiagaraSpriteRendererProperties, rendererIndex=2}
   {rendererClass=&NiagaraMeshRendererProperties, rendererIndex=3}
   {rendererClass=&NiagaraSpriteRendererProperties, rendererIndex=4}
   {rendererClass=&NiagaraSpriteRendererProperties, rendererIndex=5}
  ]
 }
}
summary {
 systemName=NS_UmbraGenesis
 emitters [
  {emitterName=NE_Sprites, rendererClasses [&NiagaraSpriteRendererProperties, &NiagaraMeshRendererProperties]}
  {emitterName=NE_Cone_Clouds, rendererClasses [&NiagaraSpriteRendererProperties]}
  {emitterName=NE_Ribbon, rendererClasses [&NiagaraRibbonRendererProperties]}
 ]
 userVariables [
  User.bIsFrontendPreview {type=NiagaraBool}
  User.bIsPainted {type=NiagaraBool}
  User.ColorVariantID {type=NiagaraInt32, defaultValue=3}
  User.DTCarOffset {type=NiagaraPosition, description="", defaultValue {}}
  User.DTCarOffset {type=Vector3f, description="", defaultValue {struct=Vector3f, value=<0 0 0>}}
  User.DTCarOffset001 {type=Vector3f, description="", defaultValue {struct=Vector3f, value=<0 0 0>}}
  User.GameModeID {type=NiagaraInt32}
  User.offset {type=NiagaraPosition, description="", defaultValue {}}
  User.PaintedColor {type=LinearColor, description="", defaultValue {struct=LinearColor, value=<rgba 1 1 1 1>}}
  User.PaintedScalar {type=NiagaraFloat, defaultValue=1, description=""}
  User.PortCount {type=NiagaraInt32, defaultValue=2}
  User.WaveMaxScale {type=NiagaraFloat, defaultValue=2, description=""}
  User.WavePeriod {type=NiagaraFloat, defaultValue=0.5, description=""}
 ]
}
